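# Translucent, unlit, two-sided Niagara sprite material that draws a soft triangular wedge glow for the Stunning Mountain pickaxe effects.
# Centered UVs convert to radial coordinates, and mirrored thresholds around the angle seam form the wedge, shaped by Angle, Hardness and Power parameters.
# A soft circular mask fades the wedge, and particle alpha scales the result.
# Color parameter C times particle color feeds the shared emissive adjustment function.
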Material /BRCosmetics/Weapons/FORT_Melee/Pickaxe_StunningMountain/FX/M_TriangleGlow
graph {
 properties {
  bUsedWithNiagaraSprites=true
  blendMode=BLEND_Translucent
  pixelDepthOffsetMode=PDOM_Legacy
  shadingModel=MSM_Unlit
  twoSided=true
  usedShadingModels=Unlit
 }
 nodes [
  n1 = VectorParameter() {parameterName=C, defaultValue=<rgba 1 1 1 1>}
  n2 = ParticleColor()
  MaterialFunctionCall(Emissive=Divide(A=Multiply(A=@n1.RGB, B=@n2.RGB))) {materialFunction=&MF_FX_EmissiveAdjustments}
  n3 = SphereMask(A=TextureCoordinate(), B=Constant2Vector() {r=0.5}) {attenuationRadius=1, hardnessPercent=0}
  n4 = Add(A=TextureCoordinate() {uTiling=2, vTiling=2}, B=Constant2Vector() {r=-1}) {constB=-1}
  n5 = MaterialFunctionCall(Float2=MaterialFunctionCall(Vector or UVs=MaterialFunctionCall(XY=@n4) {materialFunction=&Swizzle}.YX) {materialFunction=&VectorToRadialValue}.Radial Coordinates) {materialFunction=&BreakOutFloat2Components}
  n6 = OneMinus(None=@n5.R)
  n7 = ScalarParameter() {defaultValue=0.01, parameterName=Angle}
  n8 = ScalarParameter() {defaultValue=18, parameterName=Hardness}
  n9 = Saturate(None=Multiply(A=Subtract(A=@n6, B=@n7), B=@n8))
  n10 = Saturate(None=Multiply(A=Subtract(A=@n5.R, B=@n7), B=@n8))
  n11 = OneMinus(None=Saturate(None=Multiply(A=@n9, B=@n10)))
  n12 = ScalarParameter() {defaultValue=1, parameterName=Power}
  Multiply(A=@n2.A, B=Multiply(A=Power(Base=@n3) {constExponent=1.5}, B=Power(Base=@n11, Exp=@n12)))
 ]
}
