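# Animation Blueprint for the StunningMountain pickaxe that adds physics motion, hovering bolt movement and grip offsets.
# Two RigidBody simulations based on the handle bone run on the input pose, with a bool blend switching to the jumping physics asset while jumping or falling.
# Real-time sine waves drive additive translations on the front and back bolt bones.
# Fixed offsets then adjust handle and front bolt, with extra handle offset for Large body types.
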AnimBlueprint /BRCosmetics/Weapons/FORT_Melee/Pickaxe_StunningMountain/Meshes/StunningMountain_Axe_AnimBP
target_skeleton=&/StunningMountain_Axe_Skeleton
graphs [
 AnimGraph {nodes [
   n1 = "Transform (Modify) Bone - Bone: base_bolt_bk_r"(
    ComponentPose=Local To Component(LocalPose=Use cached pose 'CRCache_dyn'())
    Alpha=1
    Translation=Get backBolt() -> (backBolt:Vector)
    Rotation="0,0,0"
    Scale="1,1,1"
   ) {translationMode=BMM_Additive, translationSpace=BCS_BoneSpace, alphaBoolBlend {blendOption=Linear}, boneToModify {boneName=base_bolt_bk_r}}
   n2 = "Transform (Modify) Bone - Bone: base_bolt_fr_r"(ComponentPose=@n1, Alpha=1, Translation=Get frontBolt() -> (frontBolt:Vector), Rotation="0,0,0", Scale="1,1,1") {translationMode=BMM_Additive, translationSpace=BCS_BoneSpace, alphaBoolBlend {blendOption=Linear}, boneToModify {boneName=base_bolt_fr_r}}
   n3 = "Transform (Modify) Bone - Bone: Melee_TwoHanded_Handle"(ComponentPose=@n2, Alpha=1, Translation="0.5,0,0", Rotation="0,0,0", Scale="1,1,1") {
    translationMode=BMM_Additive
    translationSpace=BCS_BoneSpace
    alphaBoolBlend {blendOption=Linear}
    boneToModify {boneName=Melee_TwoHanded_Handle}
    translation {x=0.5}
   }
   n4 = "Transform (Modify) Bone - Bone: base_bolt_fr_r"(ComponentPose=@n3, Alpha=1, Translation="-0.5,-1.5,0", Rotation="0,0,0", Scale="1,1,1") {
    translationMode=BMM_Additive
    translationSpace=BCS_BoneSpace
    alphaBoolBlend {blendOption=Linear}
    boneToModify {boneName=base_bolt_fr_r}
    translation=<-0.5 -1.5>
   }
   n5 = "Transform (Modify) Bone - Bone: Melee_TwoHanded_Handle"(
    ComponentPose=@n4
    Alpha="To Float (Boolean)"(InBool=Get IsLarge() -> (IsLarge:bool))
    Translation="1.5,0,0"
    Rotation="0,0,0"
    Scale="1,1,1"
   ) {
    translationMode=BMM_Additive
    translationSpace=BCS_BoneSpace
    alphaBoolBlend {blendOption=Linear}
    boneToModify {boneName=Melee_TwoHanded_Handle}
    translation {x=1.5}
   }
   Output Pose(Result=Component To Local(ComponentPose=@n5)) {layerGroup=None}
   n6 = RigidBody(ComponentPose=Local To Component(LocalPose=Use cached pose 'CRCache_Input'())) {
    bOverrideWorldGravity=true
    overridePhysicsAsset=&/StunningMountain_Axe_Physics
    simulationSpace=BaseBoneSpace
    alphaBoolBlend {blendOption=Linear}
    baseBoneRef {boneName=Melee_TwoHanded_Handle}
    componentAppliedLinearAccClamp=<3000 3000 100>
    componentLinearAccScale=<0.2 0.2 0.05>
    componentLinearVelScale=<0.2 0.2 0.05>
    simSpaceSettings {
     maxAngularAcceleration=1000
     maxAngularVelocity=3
     maxLinearAcceleration=700
     maxLinearVelocity=500
     velocityScaleZ=0
     worldAlpha=0.3
     externalLinearDragV=<10 10 10>
    }
   }
   n7 = RigidBody(ComponentPose=Local To Component(LocalPose=Use cached pose 'CRCache_Input'())) {
    bOverrideWorldGravity=true
    overridePhysicsAsset=&/StunningMountain_Jumping_Axe_Physics
    simulationSpace=BaseBoneSpace
    alphaBoolBlend {blendOption=Linear}
    baseBoneRef {boneName=Melee_TwoHanded_Handle}
    componentAppliedLinearAccClamp=<3000 3000 100>
    componentLinearAccScale=<0.15 0.15 0.01>
    componentLinearVelScale=<0.15 0.15 0.01>
    simSpaceSettings {
     maxAngularAcceleration=1000
     maxAngularVelocity=3
     maxLinearAcceleration=700
     maxLinearVelocity=500
     velocityScaleZ=0
     worldAlpha=0.3
     externalLinearDragV=<10 10 10>
    }
   }
   Save cached pose 'CRCache_dyn'(Pose=Blend Poses by bool(
     BlendPose_0=Component To Local(ComponentPose=@n7)
     BlendPose_1=Component To Local(ComponentPose=@n6)
     bActiveValue=Get jumpAlpha() -> (jumpAlpha:bool)
    ) {blendType=Linear}) {cacheName=CRCache_dyn}
   AnimGraphNode_SubInput_0 = Input Pose()
   Save cached pose 'CRCache_Input'(Pose=@AnimGraphNode_SubInput_0) {cacheName=CRCache_Input}
  ]}
 EventGraph {nodes [
   Event BlueprintUpdateAnimation()
   > n1 = Sequence!(execute) -> (then_0, then_1, then_2:Exec)
   n2 = GetRealTimeSeconds()
   n3 = "Sin (Radians)"(A=float * float(A=float * float(A=1.3, B=1.3), B=@n2))
   n4 = "Sin (Radians)"(A=float * float(A=float * float(A=0.9, B=1), B=@n2))
   n5 = "Sin (Radians)"(A=float * float(A=float * float(A=0.8, B=1.8), B=@n2))
   Set frontBolt(
    execute=@n1.then_1
    frontBolt:Vector=MakeVector(
     X=Remap Float(Input=@n5, OldRange=(X=-1,Y=1), NewRange=(X=-0.3,Y=0.3)) -> (Output=~)
     Y=float * float(A=@n3, B=1)
     Z=float * float(A=@n4, B=0.1)
    )
   ) -> (Output_Get:Vector)
   n6 = GetRealTimeSeconds()
   n7 = "Sin (Radians)"(A=float * float(A=float * float(A=0.9, B=1), B=@n6))
   n8 = "Sin (Radians)"(A=float * float(A=float * float(A=1.1, B=1.1), B=@n6))
   n9 = "Sin (Radians)"(A=float * float(A=float * float(A=2, B=1.5), B=@n6))
   Set backBolt(
    execute=@n1.then_0
    backBolt:Vector=MakeVector(
     X=Remap Float(Input=@n9, OldRange=(X=-1,Y=1), NewRange=(X=-0.3,Y=1)) -> (Output=~)
     Y=float * float(A=@n8, B=1)
     Z=float * float(A=@n7, B=0.1)
    )
   ) -> (Output_Get:Vector)
   n10 = Branch(execute=@n1.then_2, Condition=OR Boolean(A=Get bIsJumping(), B=Get bIsFalling()))
   > Set jumpAlpha(jumpAlpha:bool=1) -> (Output_Get:bool)
   Set jumpAlpha(execute=@n10.else, jumpAlpha=false)
   Event OnCharacterCustomizationLoaded()
   > n11 = Set Owning Pawn(Owning Pawn:Pawn Object Reference=TryGetPawnOwner()) -> (Output_Get:Pawn Object Reference)
   > Is Valid(InputObject=@n11)
   > n12 = Sequence()
   n13 = FindBodyType!(execute=@n12.then_1, Object:Object Reference=Get Owning Pawn() -> (Owning Pawn:Pawn Object Reference)) -> (True:Exec, False:Exec, OutBodyType:EFortCustomBodyType Enum=NONE)
   > Set IsLarge(IsLarge:bool="Equal (Enum)"(A:EFortCustomBodyType Enum=@n13, B:EFortCustomBodyType Enum=Large)) -> (Output_Get:bool)
   TryGetPawnOwner()
  ]}
]
variables [
 backBolt {type=Vector, default=<0.5793 -0.1552 -0.0868>}
 frontBolt {type=Vector, default=<-0.0608 0.0878 -0.0868>}
 IsLarge {type=boolean}
 jumpAlpha {type=boolean}
 Owning Pawn {type=Pawn, default=null}
]
